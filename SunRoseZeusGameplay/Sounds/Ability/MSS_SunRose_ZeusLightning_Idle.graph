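# Looping stereo MetaSound source for the idle sound of the SunRose Zeus Lightning ability, with attenuation and the ranged weapon sound class.
# On play, it fires a random appear one-shot and starts the idle loop.
# Each step trigger fires a zero-sustain ADSR that modulates the loop's pitch and gain, plus a random step sound player whose array defaults to empty.
# A mixer sums all three, with the loop attenuated; the OnJump input is unused.
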MetaSoundSource /SunRoseZeusGameplay/Sounds/Ability/MSS_SunRose_ZeusLightning_Idle
inputs [
 Trigger.OnJump {type=Float}
 Trigger.OnStep {type=Trigger}
 UE.Source.OnPlay {type=Trigger}
 Waves.IdleLoop {type=WaveAsset:Array}
 Waves.OnAppear {type=WaveAsset:Array}
 Waves.Step {type=WaveAsset:Array}
]
interfaces [UE.OutputFormat.Stereo:1.0, UE.Source:1.1]
outputs [
 "UE.OutputFormat.Stereo.Audio:0" {type=Audio}
 "UE.OutputFormat.Stereo.Audio:1" {type=Audio}
]
pages [Default {
  input_defaults {
   Trigger.OnJump=0
   Trigger.OnStep=null
   UE.Source.OnPlay=null
   Waves.IdleLoop [/SunRoseZeusGameplay/Sounds/Ability/Wavs/Sunrose_ZeusLightning_Proj_Loop_01]
   Waves.OnAppear [
    /SunRoseZeusGameplay/Sounds/Ability/Wavs/Sunrose_ZeusLightning_Equip_01
    /SunRoseZeusGameplay/Sounds/Ability/Wavs/Sunrose_ZeusLightning_Equip_02
    /SunRoseZeusGameplay/Sounds/Ability/Wavs/Sunrose_ZeusLightning_Equip_03
   ]
   Waves.Step []
  }
  nodes [
   ADSREnvelopeADSREnvelope_0 = ADSR Envelope.ADSR Envelope:Float(Trigger Attack=^Trigger.OnStep) {also {defaults {Attack Time=0.2, Decay Curve=0.5, Decay Time=0.55, Sustain Level=0}}}
   n1 = &MSP_Audio_WavePlayer_Deluxe_Stereo(
    Play Trigger=^UE.Source.OnPlay
    Wave Asset=^Waves.IdleLoop
    "Pitch Shift Global (Semitones)"=/FortMetasound/SoundCue/MSP_ContinuousModulator(Value In=@ADSREnvelopeADSREnvelope_0.Out Envelope) {also {defaults {Output.Min=-1}}}.Value Out
    WavePlayer Gain dB=&MSP_ContinuousModulator(Value In=@ADSREnvelopeADSREnvelope_0.Out Envelope) {also {defaults {Output.Max=0, Output.Min=-12}}}.Value Out
   ) {also {defaults {Is Looping?="<MetaSoundBuilderResult.SUCCEEDED: 0>"}}}
   n2 = &MSS_PlayRandomSound_Stereo(UE.Source.OnPlay=^Trigger.OnStep, Sounds=^Waves.Step)
   n3 = &MSS_PlayRandomSound_Stereo(UE.Source.OnPlay=^UE.Source.OnPlay, Sounds=^Waves.OnAppear)
   AudioMixerAudioMixerStereo3_0 = "AudioMixer.Audio Mixer (Stereo, 3)"(
    In 1 L=@n1.Out L
    In 1 R=@n1.Out R
    In 2 L=@n2."UE.OutputFormat.Stereo.Audio:0"
    In 2 R=@n2."UE.OutputFormat.Stereo.Audio:1"
    In 0 L=@n3."UE.OutputFormat.Stereo.Audio:0"
    In 0 R=@n3."UE.OutputFormat.Stereo.Audio:1"
   ) {also {defaults {Gain 1=0.2}}}
  ]
  links [
   {dst="@UE.OutputFormat.Stereo.Audio:0", src=AudioMixerAudioMixerStereo3_0.Out L}
   {dst="@UE.OutputFormat.Stereo.Audio:1", src=AudioMixerAudioMixerStereo3_0.Out R}
  ]
 }]
properties {
 assetImportData=&self:AssetImportData
 attenuationSettings=&/SunRose_ZeusLightning_Idle_Att
 bLooping=true
 duration=10000
 maxDistance=4600
 numChannels=2
 outputFormat=Stereo
 sampleRate=48000
 soundClassObject=&Fort_WeaponRanged
 volume=0.5
}
