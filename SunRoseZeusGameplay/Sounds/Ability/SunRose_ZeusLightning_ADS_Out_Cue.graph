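# Sound cue for leaving aim-down-sights with the Zeus lightning ability in SunRose Zeus gameplay.
# A single wave player feeds the output directly, played quietly with a slight pitch raise under the weapon foley sound class.
# It references weapon reload attenuation and carries an owner-limited concurrency override with a small voice cap and short retrigger time.
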SoundCue /SunRoseZeusGameplay/Sounds/Ability/SunRose_ZeusLightning_ADS_Out_Cue
root=WavePlayer_1
nodes [{id=WavePlayer_1, properties {soundWaveAssetPtr=/SunRoseZeusGameplay/Sounds/Ability/Wavs/Sunrose_ZeusLightning_ADS_Out}}]
properties {
 attenuationSettings=&Fort_Weapon_Reload_Att
 bOverrideConcurrency=true
 maxDistance=2250
 pitchMultiplier=1.1
 soundClassObject=&Fort_WeaponFoley
 volumeMultiplier=0.08
 attenuationOverrides {distanceAlgorithm=NaturalSound, falloffDistance=3000, attenuationShapeExtents {x=1000}}
 concurrencyOverrides {bLimitToOwner=true, maxCount=2, retriggerTime=0.2}
}
