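# Sound cue named for the looping ambient cloud of the SunRose Zeus lightning ability.
# It has no root or sound nodes, so as it stands it plays no audio.
# It still carries reduced volume, the ranged-weapon sound class, the player melee-weapon concurrency set and natural-sound attenuation override values.
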SoundCue /SunRoseZeusGameplay/Sounds/Ability/SunRose_ZeusLightning_AmbientCloud_Loop_Cue
root=null
nodes []
properties {
 maxDistance=2097152
 soundClassObject=&Fort_WeaponRanged
 volumeMultiplier=0.25
 attenuationOverrides {distanceAlgorithm=NaturalSound, falloffDistance=3000, attenuationShapeExtents {x=1000}}
 concurrencySet [&Player_Melee_Weapon_Conc]
}
