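# Sound cue for the bolt whip whoosh-by of the SunRose Zeus lightning ability, apparently played as the projectile passes.
# A Random node picks one of five equally weighted whoosh-by waves, and a Modulator varies its pitch slightly.
# It uses a dedicated whoosh-by attenuation asset and the ranged weapon sound class.
# A concurrency override caps it at two simultaneous instances, taking precedence over the listed melee weapon concurrency set.
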SoundCue /SunRoseZeusGameplay/Sounds/Ability/SunRose_ZeusLightning_BoltWhip_Cue
root=Modulator_1
nodes [
 {id=Modulator_1, children [Random_1], properties {pitchMax=1.1, pitchMin=0.9}}
 {id=Random_1, children [WavePlayer_0,WavePlayer_4,WavePlayer_5,WavePlayer_6,WavePlayer_7], properties {weights [1,1,1,1,1]}}
 {id=WavePlayer_0, properties {soundWaveAssetPtr=/SunRoseZeusGameplay/Sounds/Ability/Wavs/Sunrose_ZeusLightning_Proj_Whoosh_By_01}}
 {id=WavePlayer_4, properties {soundWaveAssetPtr=/SunRoseZeusGameplay/Sounds/Ability/Wavs/Sunrose_ZeusLightning_Proj_Whoosh_By_02}}
 {id=WavePlayer_5, properties {soundWaveAssetPtr=/SunRoseZeusGameplay/Sounds/Ability/Wavs/Sunrose_ZeusLightning_Proj_Whoosh_By_03}}
 {id=WavePlayer_6, properties {soundWaveAssetPtr=/SunRoseZeusGameplay/Sounds/Ability/Wavs/Sunrose_ZeusLightning_Proj_Whoosh_By_04}}
 {id=WavePlayer_7, properties {soundWaveAssetPtr=/SunRoseZeusGameplay/Sounds/Ability/Wavs/Sunrose_ZeusLightning_Proj_Whoosh_By_05}}
]
properties {
 attenuationSettings=&/SunRose_ZeusLightning_WhooshBy_Att
 bOverrideConcurrency=true
 maxDistance=9500
 soundClassObject=&Fort_WeaponRanged
 volumeMultiplier=1
 attenuationOverrides {distanceAlgorithm=NaturalSound, falloffDistance=3000, attenuationShapeExtents {x=1000}}
 concurrencyOverrides {maxCount=2}
 concurrencySet [&Player_Melee_Weapon_Conc]
}
