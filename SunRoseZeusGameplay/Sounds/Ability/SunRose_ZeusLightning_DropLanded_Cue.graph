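# Sound cue that plays when the Zeus Lightning ability item lands after being dropped, as its name indicates.
# A single landing wave feeds a Modulator with default settings, at reduced cue volume under the loot-collect sound class.
# It references weapon-drop attenuation, and its overridden concurrency caps simultaneous instances at a few, preventing new ones beyond that.
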SoundCue /SunRoseZeusGameplay/Sounds/Ability/SunRose_ZeusLightning_DropLanded_Cue
root=Modulator_0
nodes [
 {id=Modulator_0, children [WavePlayer_0]}
 {id=WavePlayer_0, properties {soundWaveAssetPtr=/SunRoseZeusGameplay/Sounds/Ability/Wavs/Sunrose_Zeus_Lightning_Land_01}}
]
properties {
 attenuationSettings=&Fort_Weapon_Drop_Att
 bOverrideConcurrency=true
 maxDistance=3000
 soundClassObject=&Fort_Loot_Collect
 volumeMultiplier=0.4
 attenuationOverrides {falloffDistance=1000, attenuationShapeExtents {x=0}}
 concurrencyOverrides {maxCount=3, resolutionRule=PreventNew}
}
