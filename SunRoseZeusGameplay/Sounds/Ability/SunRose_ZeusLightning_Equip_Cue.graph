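# Sound cue for equipping the SunRose Zeus Lightning ability, classed as weapon foley.
# A Random node picks one of three equally weighted equip waves and feeds a root Modulator at default settings for variation.
# It plays at reduced volume with the shared Fortnite weapon reload attenuation asset.
# Its own concurrency override caps simultaneous plays per owner and sets a short retrigger time.
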SoundCue /SunRoseZeusGameplay/Sounds/Ability/SunRose_ZeusLightning_Equip_Cue
root=Modulator_0
nodes [
 {id=Modulator_0, children [Random_1]}
 {id=Random_1, children [WavePlayer_4,WavePlayer_5,WavePlayer_6], properties {weights [1,1,1]}}
 {id=WavePlayer_4, properties {soundWaveAssetPtr=/SunRoseZeusGameplay/Sounds/Ability/Wavs/Sunrose_ZeusLightning_Equip_03}}
 {id=WavePlayer_5, properties {soundWaveAssetPtr=/SunRoseZeusGameplay/Sounds/Ability/Wavs/Sunrose_ZeusLightning_Equip_02}}
 {id=WavePlayer_6, properties {soundWaveAssetPtr=/SunRoseZeusGameplay/Sounds/Ability/Wavs/Sunrose_ZeusLightning_Equip_01}}
]
properties {
 attenuationSettings=&Fort_Weapon_Reload_Att
 bOverrideConcurrency=true
 maxDistance=2250
 soundClassObject=&Fort_WeaponFoley
 volumeMultiplier=0.376
 attenuationOverrides {distanceAlgorithm=NaturalSound, falloffDistance=3000, attenuationShapeExtents {x=1000}}
 concurrencyOverrides {bLimitToOwner=true, maxCount=2, retriggerTime=0.2}
}
