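# Sound cue for the large impact of the Zeus lightning ability, layering close, medium and distant sounds by attenuation.
# The close layer mixes an impact wave with randomized dirt-and-stone explosion debris under a 3P sound class.
# Medium and distant layers randomize between two waves each under a distant-fire class.
# Branches on islocal and isenemy share that mix, halving volume only when both are false.
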SoundCue /SunRoseZeusGameplay/Sounds/Ability/SunRose_ZeusLightning_Impact_Large_Cue
root=Branch_0
nodes [
 {id=Attenuation_0, children [Random_3], properties {attenuationSettings=/Game/Sounds/Attenuation/Fort_Explosion_Close_Athena_Att}}
 {
  id=Attenuation_118
  children [Mixer_1]
  properties {
   attenuationSettings=&/Sunrose_Zeus_Lightning_Impt_Close_Att
   attenuationOverrides {stereoSpread=29871.3984}
  }
 }
 {
  id=Attenuation_119
  children [Random_4]
  properties {attenuationSettings=&/Sunrose_Zeus_Lightning_Impt_Med_Att}
 }
 {
  id=Attenuation_120
  children [Random_5]
  properties {attenuationSettings=&/Sunrose_Zeus_Lightning_Impt_Dist_Att}
 }
 {id=Branch_0, children [Modulator_2,Branch_3,Modulator_2], properties {boolParameterName=islocal}}
 {id=Branch_3, children [Modulator_4,Modulator_3,Modulator_4], properties {boolParameterName=isenemy}}
 {id=Mixer_0, children [SoundClass_77, SoundClass_76], properties {inputVolume [1,1.5]}}
 {id=Mixer_1, children [Random_2, Attenuation_0], properties {inputVolume [1.5,0.5]}}
 {id=Mixer_2, children [Attenuation_119, Attenuation_120], properties {inputVolume [1,1]}}
 {id=Modulator_2, children [Mixer_0], properties {pitchMax=1, pitchMin=1, volumeMax=1, volumeMin=1}}
 {id=Modulator_3, children [Mixer_0], properties {pitchMax=1, pitchMin=1, volumeMax=0.5, volumeMin=0.5}}
 {id=Modulator_4, children [Mixer_0], properties {pitchMax=1, pitchMin=1, volumeMax=1, volumeMin=1}}
 {id=Random_2, children [WavePlayer_3], properties {weights [1]}}
 {id=Random_3, children [WavePlayer_1, WavePlayer_4, WavePlayer_10, WavePlayer_11, WavePlayer_12], properties {weights [1,1,1,1,1]}}
 {id=Random_4, children [WavePlayer_9, WavePlayer_13], properties {weights [1,1]}}
 {id=Random_5, children [WavePlayer_14, WavePlayer_15], properties {weights [1,1]}}
 {id=SoundClass_76, children [Mixer_2], properties {soundClassOverride=&Fort_WeaponRanged_Fire_Distant}}
 {
  id=SoundClass_77
  children [Attenuation_118]
  properties {soundClassOverride=&Fort_WeaponRanged_Fire_3P_PSM}
 }
 {id=WavePlayer_1, properties {soundWaveAssetPtr=/Game/Athena/Sounds/Weapons/Explosions/EXPL_Debris_Dirt_Stone_01}}
 {id=WavePlayer_10, properties {soundWaveAssetPtr=/Game/Athena/Sounds/Weapons/Explosions/EXPL_Debris_Dirt_Stone_05}}
 {id=WavePlayer_11, properties {soundWaveAssetPtr=/Game/Athena/Sounds/Weapons/Explosions/EXPL_Debris_Dirt_Stone_02}}
 {id=WavePlayer_12, properties {soundWaveAssetPtr=/Game/Athena/Sounds/Weapons/Explosions/EXPL_Debris_Dirt_Stone_04}}
 {id=WavePlayer_13, properties {soundWaveAssetPtr=/SunRoseZeusGameplay/Sounds/Ability/Wavs/Sunrose_ZeusLightning_Impact_Med_05}}
 {id=WavePlayer_14, properties {soundWaveAssetPtr=/SunRoseZeusGameplay/Sounds/Ability/Wavs/Sunrose_ZeusLightning_Impact_Dist_04}}
 {id=WavePlayer_15, properties {soundWaveAssetPtr=/SunRoseZeusGameplay/Sounds/Ability/Wavs/Sunrose_ZeusLightning_Impact_Dist_05}}
 {id=WavePlayer_3, properties {soundWaveAssetPtr=/SunRoseZeusGameplay/Sounds/Ability/Wavs/Sunrose_ZeusLightning_Impact_Close_03}}
 {id=WavePlayer_4, properties {soundWaveAssetPtr=/Game/Athena/Sounds/Weapons/Explosions/EXPL_Debris_Dirt_Stone_03}}
 {id=WavePlayer_9, properties {soundWaveAssetPtr=/SunRoseZeusGameplay/Sounds/Ability/Wavs/Sunrose_ZeusLightning_Impact_Med_04}}
]
properties {maxDistance=30000, soundClassObject=&Fort_WeaponRanged, volumeMultiplier=1.25}
