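# Sound cue for the large water impact of the Zeus lightning ability, using general lightning impact waves.
# Branches on islocal and isenemy halve the volume for non-local, non-enemy instigators and keep others at full.
# All paths share one mix of close, medium and distant layers, each with its own explosion attenuation.
# Medium and distant layers randomize waves under the distant sound class, while close plays a single wave under a third-person class.
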SoundCue /SunRoseZeusGameplay/Sounds/Ability/SunRose_ZeusLightning_Impact_Water_Large_Cue
root=Branch_0
nodes [
 {
  id=Attenuation_1
  children [Random_4]
  properties {attenuationSettings=/Game/Sounds/Attenuation/Fort_Explosion_Close_Athena_Att, attenuationOverrides {stereoSpread=29871.3984}}
 }
 {id=Attenuation_120, children [Random_3], properties {attenuationSettings=/Game/Sounds/Attenuation/Fort_Explosion_Distant_Athena_Att}}
 {id=Attenuation_2, children [Random_2], properties {attenuationSettings=/Game/Sounds/Attenuation/Fort_Explosion_Medium_Athena_Att}}
 {id=Branch_0, children [Modulator_1,Branch_3,Modulator_1], properties {boolParameterName=islocal}}
 {id=Branch_3, children [Modulator_6,Modulator_5,Modulator_6], properties {boolParameterName=isenemy}}
 {id=Mixer_1, children [SoundClass_1, SoundClass_76], properties {inputVolume [1,1]}}
 {id=Mixer_2, children [Attenuation_2, Attenuation_120], properties {inputVolume [1,1]}}
 {id=Modulator_1, children [Mixer_1], properties {pitchMax=1, pitchMin=1, volumeMax=1, volumeMin=1}}
 {id=Modulator_5, children [Mixer_1], properties {pitchMax=1, pitchMin=1, volumeMax=0.5, volumeMin=0.5}}
 {id=Modulator_6, children [Mixer_1], properties {pitchMax=1, pitchMin=1, volumeMax=1, volumeMin=1}}
 {id=Random_2, children [WavePlayer_1,WavePlayer_6,WavePlayer_7,WavePlayer_8], properties {weights [1,1,1,1]}}
 {id=Random_3, children [WavePlayer_10, WavePlayer_11, WavePlayer_12, WavePlayer_19], properties {weights [1,1,1,1]}}
 {id=Random_4, children [WavePlayer_28], properties {weights [1]}}
 {
  id=SoundClass_1
  children [Attenuation_1]
  properties {soundClassOverride=&Fort_WeaponRanged_Fire_3P_PSM}
 }
 {id=SoundClass_76, children [Mixer_2], properties {soundClassOverride=&Fort_WeaponRanged_Fire_Distant}}
 {id=WavePlayer_1, properties {soundWaveAssetPtr=/SunRoseZeusGameplay/Sounds/Ability/Wavs/Sunrose_ZeusLightning_Impact_Med_01}}
 {id=WavePlayer_10, properties {soundWaveAssetPtr=/SunRoseZeusGameplay/Sounds/Ability/Wavs/Sunrose_ZeusLightning_Impact_Dist_05}}
 {id=WavePlayer_11, properties {soundWaveAssetPtr=/SunRoseZeusGameplay/Sounds/Ability/Wavs/Sunrose_ZeusLightning_Impact_Dist_04}}
 {id=WavePlayer_12, properties {soundWaveAssetPtr=/SunRoseZeusGameplay/Sounds/Ability/Wavs/Sunrose_ZeusLightning_Impact_Dist_03}}
 {id=WavePlayer_19, properties {soundWaveAssetPtr=/SunRoseZeusGameplay/Sounds/Ability/Wavs/Sunrose_ZeusLightning_Impact_Dist_02}}
 {id=WavePlayer_28, properties {soundWaveAssetPtr=/SunRoseZeusGameplay/Sounds/Ability/Wavs/Sunrose_ZeusLightning_Impact_Close_03}}
 {id=WavePlayer_6, properties {soundWaveAssetPtr=/SunRoseZeusGameplay/Sounds/Ability/Wavs/Sunrose_ZeusLightning_Impact_Med_02}}
 {id=WavePlayer_7, properties {soundWaveAssetPtr=/SunRoseZeusGameplay/Sounds/Ability/Wavs/Sunrose_ZeusLightning_Impact_Med_03}}
 {id=WavePlayer_8, properties {soundWaveAssetPtr=/SunRoseZeusGameplay/Sounds/Ability/Wavs/Sunrose_ZeusLightning_Impact_Med_04}}
]
properties {maxDistance=27500, soundClassObject=&Fort_WeaponRanged, volumeMultiplier=0.2}
