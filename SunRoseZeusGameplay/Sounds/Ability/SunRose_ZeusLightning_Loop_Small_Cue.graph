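# Sound cue for the small looping Zeus lightning projectile sound in SunRoseZeusGameplay.
# A single looping wave with randomized pitch feeds volume modulators selected by nested islocal and isenemy branches.
# Non-local friendly sources play at half the level of local, enemy or unset ones.
# It uses the ranged weapon sound class, a low overall volume and close projectile attenuation, with natural-sound falloff overrides also stored.
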SoundCue /SunRoseZeusGameplay/Sounds/Ability/SunRose_ZeusLightning_Loop_Small_Cue
root=Branch_2
nodes [
 {id=Branch_2, children [Modulator_2,Branch_3,Modulator_2], properties {boolParameterName=islocal}}
 {id=Branch_3, children [Modulator_4,Modulator_3,Modulator_4], properties {boolParameterName=isenemy}}
 {id=Modulator_1, children [WavePlayer_1], properties {pitchMax=1.2, pitchMin=0.85}}
 {id=Modulator_2, children [Modulator_1], properties {pitchMax=1, pitchMin=1, volumeMax=1, volumeMin=1}}
 {id=Modulator_3, children [Modulator_1], properties {pitchMax=1, pitchMin=1, volumeMax=0.5, volumeMin=0.5}}
 {id=Modulator_4, children [Modulator_1], properties {pitchMax=1, pitchMin=1, volumeMax=1, volumeMin=1}}
 {id=WavePlayer_1, properties {bLooping=true, soundWaveAssetPtr=/SunRoseZeusGameplay/Sounds/Ability/Wavs/Sunrose_ZeusLightning_Proj_Loop_01}}
]
properties {
 attenuationSettings=&Fort_Weapon_Projectile_Close_Athena_Att
 maxDistance=4950
 soundClassObject=&Fort_WeaponRanged
 volumeMultiplier=0.1
 attenuationOverrides {distanceAlgorithm=NaturalSound, falloffDistance=3000, attenuationShapeExtents {x=1000}}
}
