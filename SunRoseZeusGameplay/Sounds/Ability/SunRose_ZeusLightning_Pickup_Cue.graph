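# Sound cue that plays the pickup sound for the SunRose Zeus Lightning ability item.
# A single wave player feeds a root Modulator left at its default settings.
# It uses the shared Fortnite weapon pickup attenuation and loot-collect sound class.
# It references the PickupMaximum_Oldest concurrency set and also carries single-instance, prevent-new override values.
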SoundCue /SunRoseZeusGameplay/Sounds/Ability/SunRose_ZeusLightning_Pickup_Cue
root=Modulator_5
nodes [
 {id=Modulator_5, children [WavePlayer_2]}
 {id=WavePlayer_2, properties {soundWaveAssetPtr=/SunRoseZeusGameplay/Sounds/Ability/Wavs/Sunrose_ZeusLightning_PU_01}}
]
properties {
 attenuationSettings=&Fort_Weapon_Pickup_Att
 maxDistance=3500
 soundClassObject=&Fort_Loot_Collect
 concurrencyOverrides {maxCount=1, resolutionRule=PreventNew}
 concurrencySet [&PickupMaximum_Oldest]
}
