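# Looping idle ambience SoundCue for the Zeus lightning ability pickup, using loot ambient sound class, attenuation and concurrency settings.
# A mixer layers a slightly pitched-up looping bed, faded in by an enveloper, with a much quieter one-shot layer.
# The one-shot branch loops through a random delay, then plays one of eight equally weighted variants with randomized pitch and volume, giving irregular accents over the steady bed.
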SoundCue /SunRoseZeusGameplay/Sounds/Ability/SunRose_ZeusLightning_Pickup_Idle_Loop_Cue
root=Mixer_0
nodes [
 {id=Delay_0, children [Modulator_1], properties {delayMax=3, delayMin=0.15}}
 {
  id=Enveloper_0
  children [Modulator_0]
  properties {pitchCurve {editorCurveData {keys [{time=0}]}}, volumeCurve {editorCurveData {keys [<at 0 0>,<at 1 1>]}}}
 }
 {id=Looping_0, children [Delay_0]}
 {id=Mixer_0, children [Enveloper_0,Looping_0], properties {inputVolume [1,0.2]}}
 {id=Modulator_0, children [WavePlayer_0], properties {pitchMax=1.2, pitchMin=1.1, volumeMax=1, volumeMin=1}}
 {id=Modulator_1, children [Random_0], properties {pitchMax=1.15, pitchMin=0.85, volumeMax=1, volumeMin=0.5}}
 {
  id=Random_0
  children [WavePlayer_1,WavePlayer_2,WavePlayer_3,WavePlayer_4,WavePlayer_6,WavePlayer_7,WavePlayer_8,WavePlayer_9]
  properties {bShouldExcludeFromBranchCulling=true, weights [1,1,1,1,1,1,1,1]}
 }
 {
  id=WavePlayer_0
  properties {bLooping=true, soundWaveAssetPtr=/SunRoseZeusGameplay/Sounds/Ability/Wavs/SunRose_ZeusLightning_Ambient_Idle_Loop}
 }
 {id=WavePlayer_1, properties {soundWaveAssetPtr=/SunRoseZeusGameplay/Sounds/Ability/Wavs/SunRose_ZeusLightning_Ambient_Idle_OneShots_01}}
 {id=WavePlayer_2, properties {soundWaveAssetPtr=/SunRoseZeusGameplay/Sounds/Ability/Wavs/SunRose_ZeusLightning_Ambient_Idle_OneShots_02}}
 {id=WavePlayer_3, properties {soundWaveAssetPtr=/SunRoseZeusGameplay/Sounds/Ability/Wavs/SunRose_ZeusLightning_Ambient_Idle_OneShots_03}}
 {id=WavePlayer_4, properties {soundWaveAssetPtr=/SunRoseZeusGameplay/Sounds/Ability/Wavs/SunRose_ZeusLightning_Ambient_Idle_OneShots_04}}
 {id=WavePlayer_6, properties {soundWaveAssetPtr=/SunRoseZeusGameplay/Sounds/Ability/Wavs/SunRose_ZeusLightning_Ambient_Idle_OneShots_05}}
 {id=WavePlayer_7, properties {soundWaveAssetPtr=/SunRoseZeusGameplay/Sounds/Ability/Wavs/SunRose_ZeusLightning_Ambient_Idle_OneShots_06}}
 {id=WavePlayer_8, properties {soundWaveAssetPtr=/SunRoseZeusGameplay/Sounds/Ability/Wavs/SunRose_ZeusLightning_Ambient_Idle_OneShots_07}}
 {id=WavePlayer_9, properties {soundWaveAssetPtr=/SunRoseZeusGameplay/Sounds/Ability/Wavs/SunRose_ZeusLightning_Ambient_Idle_OneShots_08}}
]
properties {
 attenuationSettings=/Game/Sounds/Attenuation/Fort_Item_Ambient_Loop_Mono_Default_Att
 maxDistance=1250
 priority=0.5
 soundClassObject=/Game/Sounds/FortSoundClassesAndModes/Fort_Loot_Ambient_Loop
 volumeMultiplier=0.4
 concurrencyOverrides {maxCount=4, resolutionRule=StopQuietest, voiceStealReleaseTime=0.25, volumeScale=0.3, volumeScaleAttackTime=0.5, volumeScaleMode=Distance}
 concurrencySet [/Game/Sounds/Concurrency_Rules/Loot_Ambient_Conc]
}
