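# Sound cue for the dissipate sound of the SunRose Zeus lightning rise ability, playing one wave through a quality-level node.
# Branches on islocal and isenemy keep full volume for local, enemy and unset cases, and halve it for non-local non-enemies.
# Output uses the ranged-weapon sound class, a shared lightning attenuation asset and melee-weapon concurrency.
# Only the first quality slot has audio connected.
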SoundCue /SunRoseZeusGameplay/Sounds/Ability/SunRose_ZeusLightning_Rise_Dissipate_Cue
root=Branch_2
nodes [
 {id=Branch_2, children [Modulator_2,Branch_3,Modulator_2], properties {boolParameterName=islocal}}
 {id=Branch_3, children [Modulator_4,Modulator_3,Modulator_4], properties {boolParameterName=isenemy}}
 {id=Modulator_2, children [QualityLevel_0], properties {pitchMax=1, pitchMin=1, volumeMax=1, volumeMin=1}}
 {id=Modulator_3, children [QualityLevel_0], properties {pitchMax=1, pitchMin=1, volumeMax=0.5, volumeMin=0.5}}
 {id=Modulator_4, children [QualityLevel_0], properties {pitchMax=1, pitchMin=1, volumeMax=1, volumeMin=1}}
 {id=QualityLevel_0, children [WavePlayer_1,null,null,null]}
 {id=WavePlayer_1, properties {soundWaveAssetPtr=/SunRoseZeusGameplay/Sounds/Ability/Wavs/Sunrose_ZeusLightning_Rise_Dissipate_01}}
]
properties {
 attenuationSettings=&/SunRose_ZeusLightning_Att
 maxDistance=20000
 soundClassObject=&Fort_WeaponRanged
 volumeMultiplier=0.35
 attenuationOverrides {distanceAlgorithm=NaturalSound, falloffDistance=3000, attenuationShapeExtents {x=1000}}
 concurrencySet [&Player_Melee_Weapon_Conc]
}
