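# Sound cue for the Zeus lightning strike ability in SunRoseZeusGameplay, limited by a lightning-bolt concurrency set.
# Nested islocal and isenemy branches halve the volume for non-local, non-enemy strikes and keep full volume otherwise.
# All paths share one mix of randomized close, medium and distant lightning layers, each with its own attenuation.
# The close layer uses a third-person weapon-fire sound class, while medium and distant share a distant-fire class.
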SoundCue /SunRoseZeusGameplay/Sounds/Ability/SunRose_ZeusLightning_Strike_Cue
root=Branch_1
nodes [
 {
  id=Attenuation_118
  children [Random_4]
  properties {attenuationSettings=&Fort_Weapon_Projectile_Close_Athena_Att, attenuationOverrides {stereoSpread=29871.3984}}
 }
 {
  id=Attenuation_119
  children [Random_5]
  properties {attenuationSettings=&Fort_Weapon_Projectile_Medium_Mono_Athena_Att}
 }
 {
  id=Attenuation_120
  children [Random_2]
  properties {attenuationSettings=&Fort_Weapon_Projectile_Distant_Mono_Athena_Att}
 }
 {id=Branch_1, children [Modulator_2,Branch_2,Modulator_2], properties {boolParameterName=islocal}}
 {id=Branch_2, children [Modulator_4,Modulator_3,Modulator_4], properties {boolParameterName=isenemy}}
 {id=Mixer_0, children [SoundClass_77, SoundClass_76], properties {inputVolume [0.65,1]}}
 {id=Mixer_2, children [Attenuation_119, Attenuation_120], properties {inputVolume [0.15,0.7]}}
 {id=Modulator_2, children [Mixer_0], properties {pitchMax=1, pitchMin=1, volumeMax=1, volumeMin=1}}
 {id=Modulator_3, children [Mixer_0], properties {pitchMax=1, pitchMin=1, volumeMax=0.5, volumeMin=0.5}}
 {id=Modulator_4, children [Mixer_0], properties {pitchMax=1, pitchMin=1, volumeMax=1, volumeMin=1}}
 {
  id=Random_2
  children [WavePlayer_6, WavePlayer_7, WavePlayer_8, WavePlayer_9, WavePlayer_13, WavePlayer_14, WavePlayer_15, WavePlayer_16]
  properties {weights [1,1,1,1,1,1,1,1]}
 }
 {id=Random_4, children [WavePlayer_22, WavePlayer_23, WavePlayer_24, WavePlayer_25], properties {weights [1,1,1,1]}}
 {id=Random_5, children [WavePlayer_17, WavePlayer_18], properties {weights [1,1]}}
 {id=SoundClass_76, children [Mixer_2], properties {soundClassOverride=&Fort_WeaponRanged_Fire_Distant}}
 {
  id=SoundClass_77
  children [Attenuation_118]
  properties {soundClassOverride=&Fort_WeaponRanged_Fire_3P_PSM}
 }
 {id=WavePlayer_13, properties {soundWaveAssetPtr=/Game/Sounds/Frontend_Sounds/HexMap_Distant_Lightning_Strikes_04}}
 {id=WavePlayer_14, properties {soundWaveAssetPtr=/Game/Sounds/Frontend_Sounds/HexMap_Distant_Lightning_Strikes_03}}
 {id=WavePlayer_15, properties {soundWaveAssetPtr=/Game/Sounds/Frontend_Sounds/HexMap_Distant_Lightning_Strikes_02}}
 {id=WavePlayer_16, properties {soundWaveAssetPtr=/Game/Sounds/Frontend_Sounds/HexMap_Distant_Lightning_Strikes_01}}
 {id=WavePlayer_17, properties {soundWaveAssetPtr=/Game/Athena/Sounds/Events/CrackLightning/Crack_Lightning_StrikeImpact_01}}
 {id=WavePlayer_18, properties {soundWaveAssetPtr=/Game/Athena/Sounds/Events/CrackLightning/Crack_Lightning_StrikeImpact_02}}
 {
  id=WavePlayer_22
  properties {soundWaveAssetPtr=/Game/Sounds/Fort_Monster_Storm_King/SFX/Final_SFX/StormKing_LightningSummons_Lightning_Impact_04}
 }
 {
  id=WavePlayer_23
  properties {soundWaveAssetPtr=/Game/Sounds/Fort_Monster_Storm_King/SFX/Final_SFX/StormKing_LightningSummons_Lightning_Impact_03}
 }
 {
  id=WavePlayer_24
  properties {soundWaveAssetPtr=/Game/Sounds/Fort_Monster_Storm_King/SFX/Final_SFX/StormKing_LightningSummons_Lightning_Impact_02}
 }
 {
  id=WavePlayer_25
  properties {soundWaveAssetPtr=/Game/Sounds/Fort_Monster_Storm_King/SFX/Final_SFX/StormKing_LightningSummons_Lightning_Impact_01}
 }
 {id=WavePlayer_6, properties {soundWaveAssetPtr=/Game/Sounds/Frontend_Sounds/HexMap_Distant_Lightning_Strikes_08}}
 {id=WavePlayer_7, properties {soundWaveAssetPtr=/Game/Sounds/Frontend_Sounds/HexMap_Distant_Lightning_Strikes_07}}
 {id=WavePlayer_8, properties {soundWaveAssetPtr=/Game/Sounds/Frontend_Sounds/HexMap_Distant_Lightning_Strikes_06}}
 {id=WavePlayer_9, properties {soundWaveAssetPtr=/Game/Sounds/Frontend_Sounds/HexMap_Distant_Lightning_Strikes_05}}
]
properties {
 maxDistance=27500
 soundClassObject=&Fort_WeaponRanged
 concurrencySet [&/Sunrose_ZeusLightningBolts_Conc]
}
